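# Unlit, two-sided translucent material for Niagara ribbons, drawing the Christmas ribbon effect of the TennisLeash pickaxe.
# A grayscale ribbon texture blends two brightnesses of tint times particle color, feeding an emissive adjustment function.
# TwoSidedSign drives an extrapolating lerp that darkens front faces and brightens back faces.
# The same texture builds a smoothstepped mask, cut with a hard step by the dynamic Erosion parameter for dissolving.
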Material /BRCosmetics/Weapons/FORT_Melee/Pickaxe_TennisLeash/FX/Materials/M_ChristmasRibbon
graph {
 properties {
  bCastRayTracedShadows=false
  bFullyRough=true
  bUsedWithNiagaraRibbons=true
  blendMode=BLEND_Translucent
  pixelDepthOffsetMode=PDOM_Legacy
  shadingModel=MSM_Unlit
  twoSided=true
  usedShadingModels=Unlit
 }
 nodes [
  n1 = Multiply(A=VectorParameter() {parameterName=Color1, defaultValue=<rgba 0.8229 0.8229 0.8229 0>}.RGB, B=ParticleColor().RGB)
  n2 = TextureSampleParameter2D() {
   parameterName=Texture
   samplerSource=SSM_Wrap_WorldGroupSettings
   samplerType=SAMPLERTYPE_LinearGrayscale
   texture=/BRCosmetics/Weapons/FORT_Melee/Pickaxe_TennisLeash/FX/Textures/T_TennisLeash_Ribbon
  }
  n3 = Multiply(A=@n2.R) {constB=1.5}
  n4 = LinearInterpolate(A=Multiply(A=@n1) {constB=1.2}, B=@n1, Alpha=Saturate(None=Power(Base=@n3)))
  MaterialFunctionCall(Emissive=LinearInterpolate(A=@n4, B=Multiply(A=@n4) {constB=0.7}, Alpha=TwoSidedSign())) {materialFunction=&MF_FX_EmissiveAdjustments}
  n5 = OneMinus(None=@n2.R)
  n6 = DynamicParameter() {defaultValue=<rgba 0 1 1 1>, paramNames [Erosion,Param2,Param3,Param4]}
  Multiply(A=SmoothStep(Min=Constant() {r=0.05}, Max=Constant() {r=0.1}, Value=@n2.R), B=OneMinus(None=Step(Y=@n5, X=@n6.Erosion) {constX=0.5}))
 ]
}
