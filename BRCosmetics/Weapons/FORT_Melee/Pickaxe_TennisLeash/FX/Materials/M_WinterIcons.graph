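# Masked unlit particle material for Niagara sprites and meshes that draws winter icons for the TennisLeash pickaxe FX.
# A grayscale SDF texture is thresholded by SmoothStep, with Thickness and Edge Smoothness setting the edge, then multiplied by particle alpha as a mask.
# Particle color times the Color parameter feeds MF_FX_EmissiveAdjustments for glow.
# A second branch adds a fixed SDF threshold to a UV half-split step, with no visible consumer.
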Material /BRCosmetics/Weapons/FORT_Melee/Pickaxe_TennisLeash/FX/Materials/M_WinterIcons
graph {
 properties {
  bOutputTranslucentVelocity=true
  bUsedWithNiagaraMeshParticles=true
  bUsedWithNiagaraSprites=true
  blendMode=BLEND_Masked
  opacityMaskClipValue=0.02
  pixelDepthOffsetMode=PDOM_Legacy
  shadingModel=MSM_Unlit
  usedShadingModels=Unlit
 }
 nodes [
  n1 = ParticleColor()
  MaterialFunctionCall(Emissive=Multiply(A=@n1.RGB, B=VectorParameter() {parameterName=Color, defaultValue=<rgba 1 0 0 1>}.RGB)) {materialFunction=&MF_FX_EmissiveAdjustments}
  n2 = OneMinus(None=ScalarParameter() {defaultValue=0.4, parameterName=Thickness})
  n3 = Subtract(A=@n2, B=ScalarParameter() {defaultValue=0.2, parameterName=Edge Smoothness})
  n4 = TextureSampleParameter2D() {
   parameterName=Texture
   samplerType=SAMPLERTYPE_LinearGrayscale
   texture=/BRCosmetics/Weapons/FORT_Melee/Pickaxe_TennisLeash/FX/Textures/T_TennisLeash_WinterIcons_SDF
  }
  Multiply(A=SmoothStep(Min=Saturate(None=@n3), Max=@n2, Value=@n4.R) {constMax=0.4, constMin=0.2}, B=@n1.A)
  ParticleColor()
  n5 = ComponentMask(None=TextureCoordinate()) {r=true}
  Saturate(None=Add(A=SmoothStep(Value=@n4.R) {constMax=0.6, constMin=0.5}, B=Step(X=@n5) {constX=0.5, constY=0.5}))
 ]
}
