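# Physics asset for the TennisLeash pickaxe mesh, simulating its dangling decorative bells, rings, bows and fur.
# Capsule bodies on dynamic joints are constrained to a shared base joint, with each ring chained to its own bell.
# Bells, rings and bows get limited swing and very tight twist so they sway without spinning.
# The fur body has all angular axes locked to the base, so it does not rotate relative to it.
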PhysicsAsset /BRCosmetics/Weapons/FORT_Melee/Pickaxe_TennisLeash/Meshes/TennisLeash_Axe_Physics
bodies {
 dyn_base_joint [{
   shapeType=Capsule
   length=3.8335
   radius=0.943
   shapeName=dyn_base_joint_capsule
   center=<0.4827 1.2323 10.4021>
   rotation=<rot -1.7672 -90.8406 2.335>
  }]
 dyn_bell_1_joint [{
   shapeType=Capsule
   length=4.3416
   radius=2
   shapeName=dyn_bell_1_joint_capsule
   center=<5.1605 -0.0759 -0.0397>
   rotation=<rot 0 91.5947 -89.1437>
  }]
 dyn_bell_2_joint [{
   shapeType=Capsule
   length=2.6046
   radius=2.5
   shapeName=dyn_bell_2_joint_capsule
   center=<5.1611 -0.0721 -0.0297>
   rotation=<rot 0 91.6634 -89.2865>
  }]
 dyn_bow_1_joint [{
   shapeType=Capsule
   length=2
   radius=2
   shapeName=dyn_bow_1_joint_capsule
   center=<3.6187 -0.3614 -0.1166>
   rotation=<rot -8.0516 -90.0003 -93.3268>
  }]
 dyn_bow_2_joint [{
   shapeType=Capsule
   length=1.5
   radius=2
   shapeName=dyn_bow_2_joint_capsule
   center=<3.6783 -0.1387 -0.2427>
   rotation=<rot 5.5314 -89.9998 -84.7345>
  }]
 dyn_bow_3_joint [{
   shapeType=Capsule
   length=5
   radius=1.5
   shapeName=dyn_bow_3_joint_capsule
   center=<4.4904 -0.241 0.1193>
   rotation=<rot 0 -87.2319 -88.2117>
  }]
 dyn_bow_4_joint [{shapeType=Capsule, length=5, radius=1, shapeName=dyn_bow_4_joint_capsule, center=<4.0651 -0.1826 0.0641>, rotation=<rot 0 91.0101 -88.9538>}]
 dyn_fur_joint [{
   shapeType=Capsule
   length=7.0151
   radius=6.9789
   shapeName=dyn_fur_joint_capsule
   center=<-0.0735 -1.0985 0.6146>
   rotation=<rot -0.5601 96.3099 -92.0951>
  }]
 dyn_ring_1_joint [{
   shapeType=Capsule
   length=5.5
   radius=0.5
   shapeName=dyn_ring_1_joint_capsule
   center=<5.1268 0.0137 -0.0075>
   rotation=<rot 0 89.6273 -89.7854>
  }]
 dyn_ring_2_joint [{
   shapeType=Capsule
   length=5.5
   radius=0.5
   shapeName=dyn_ring_2_joint_capsule
   center=<4.8757 0.0139 -0.0086>
   rotation=<rot 0 89.6144 -89.7446>
  }]
}
constraints [
 {bone1Name=dyn_bell_1_joint, bone2Name=dyn_base_joint, swing1LimitDegrees=12, swing2LimitDegrees=12, twistLimitDegrees=1}
 {bone1Name=dyn_ring_1_joint, bone2Name=dyn_bell_1_joint, swing1LimitDegrees=22, swing2LimitDegrees=22, twistLimitDegrees=1}
 {bone1Name=dyn_bell_2_joint, bone2Name=dyn_base_joint, swing1LimitDegrees=10, swing2LimitDegrees=10, twistLimitDegrees=1}
 {bone1Name=dyn_ring_2_joint, bone2Name=dyn_bell_2_joint, swing1LimitDegrees=22, swing2LimitDegrees=22, twistLimitDegrees=1}
 {bone1Name=dyn_bow_1_joint, bone2Name=dyn_base_joint, swing1LimitDegrees=10, swing2LimitDegrees=20, twistLimitDegrees=1}
 {bone1Name=dyn_bow_2_joint, bone2Name=dyn_base_joint, swing1LimitDegrees=10, swing2LimitDegrees=20, twistLimitDegrees=1}
 {bone1Name=dyn_bow_3_joint, bone2Name=dyn_base_joint, swing1LimitDegrees=10, swing2LimitDegrees=20, twistLimitDegrees=1}
 {bone1Name=dyn_bow_4_joint, bone2Name=dyn_base_joint, swing1LimitDegrees=10, swing2LimitDegrees=20, twistLimitDegrees=1}
 {
  bone1Name=dyn_fur_joint
  bone2Name=dyn_base_joint
  swing1LimitDegrees=45
  swing1Motion=Locked
  swing2LimitDegrees=45
  swing2Motion=Locked
  twistLimitDegrees=45
  twistMotion=Locked
 }
]
